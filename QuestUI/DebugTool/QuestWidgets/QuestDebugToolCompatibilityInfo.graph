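# Quest debug tool widget that lays out a Compatibility section for a quest's product compatibility with the active playlist.
# A vertical box stacks a section header and info fields for playlist product tag, compatibility query, additional query tags and a mismatch alert.
# Helper functions convert gameplay tags to debug text and return collapsed visibility for empty tag text or a compatible quest.
# The event graph is empty and the file holds no variables.
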WidgetBlueprint /QuestUI/DebugTool/QuestWidgets/QuestDebugToolCompatibilityInfo
graphs [
 EventGraph {links [], nodes []}
 GetAdditionalQueryTagVisibility {nodes [
   n1 = GetAdditionalQueryTagVisibility!() -> (then, AdditionalQueryTags:Text) {is=title K2Node_FunctionEntry}
   > Return Node!(execute, Visibility:ESlateVisibility Enum=Select(Option 0=HitTestInvisible, Option 1=Collapsed, Index=TextIsEmpty(InText=@n1)))
  ]}
 GetQueryAlertVisibility {nodes [
   n1 = GetQueryAlertVisibility!() -> (then, IsCompatible:bool=~) {is=title K2Node_FunctionEntry}
   > Return Node(Visibility=Select(Option 0=HitTestInvisible, Option 1=Collapsed, Index=@n1))
  ]}
 TagToText {nodes [
   n1 = TagToText!() -> (then, Tag:"Gameplay Tag Structure (by ref)") {is=title K2Node_FunctionEntry}
   > Return Node!(
    execute
    Text:Text="To Text (String)"(InString=GetDebugStringFromGameplayTag!(GameplayTag:Gameplay Tag Structure=@n1) -> (ReturnValue:String))
   )
  ]}
]
variables []
widgets {
 info {inheritedWidgetCount=0, parentClass=&UserWidget, rootWidgetClass=&VerticalBox, widgetCount=6}
 widgets [
  VerticalBox_79 {widgetClassPath=&VerticalBox, widget=VerticalBox_79}
  SectionHeader {
   widgetClassPath=/QuestUI/DebugTool/CommonWidgets/QuestDebugToolSectionHeader
   bIsVariable=true
   parent=VerticalBox_79
   widget=SectionHeader
   properties {header=Compatibility}
   slot {class=VerticalBoxSlot}
  }
  Widget_ActiveProduct {
   widgetClassPath=/QuestUI/DebugTool/CommonWidgets/QuestDebugToolInfoField
   bIsVariable=true
   parent=VerticalBox_79
   widget=Widget_ActiveProduct
   properties {header=Active Playlist Product Tag, multiline=true}
   slot {class=VerticalBoxSlot}
  }
  Widget_Compatibility {
   widgetClassPath=&QuestDebugToolInfoField
   bIsVariable=true
   parent=VerticalBox_79
   widget=Widget_Compatibility
   properties {header=Product Compatibility Query, multiline=true}
   slot {class=VerticalBoxSlot}
  }
  Widget_AdditionalTags {
   widgetClassPath=&QuestDebugToolInfoField
   bIsVariable=true
   parent=VerticalBox_79
   widget=Widget_AdditionalTags
   properties {header=Additional Query Tags}
   slot {class=VerticalBoxSlot, properties {size {sizeRule=Fill, value=1}}}
  }
  Widget_CompatibilityAlert {
   widgetClassPath=&QuestDebugToolInfoField
   bIsVariable=true
   parent=VerticalBox_79
   widget=Widget_CompatibilityAlert
   properties {data="Quest not compatible with active playlist!", displayAsAlert=true, header=Compatibility Alert}
   slot {class=VerticalBoxSlot, properties {size {sizeRule=Fill, value=1}}}
  }
 ]
}
