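# Quest debug tool widget that shows a titled field of category groups for a quest's data.
# PreConstruct writes the Header variable into the header text.
# SetData takes an array of category structs, resets a dynamic entry box and creates one QuestDebugToolCategoryGroup entry per element, passing each its data.
# A visibility switcher shows that list, or a notice that the quest's tags match no existing categories when the array is empty.
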WidgetBlueprint /QuestUI/DebugTool/CategoryWidgets/QuestDebugToolCategoryField
graphs [
 EventGraph {nodes [
   Event PreConstruct()
   > "SetText (Text)"(self=Get Text_Header() -> (Text_Header:Common Text Object Reference), InText=Get Header() -> (Header:Text))
  ]}
 SetData {nodes [
   n1 = SetData!() -> (then, Data:Array of Quest Panel Debug Category Data Structures) {is=title K2Node_FunctionEntry}
   > n2 = Branch(Condition="integer > integer"(A=Length(TargetArray:Array of Quest Panel Debug Category Data Structures=@n1)))
   > SetActiveWidget(
    self=Get Switcher_Content() -> (Switcher_Content:Common Visibility Switcher Object Reference)
    Widget=Get EntryBox_Products() -> (EntryBox_Products:Dynamic Entry Box Object Reference)
   )
   > Reset!(execute, self:Dynamic Entry Box Object Reference=Get EntryBox_Products(), bDeleteWidgets:bool=false) -> (then)
   > n3 = For Each Loop(Array:Array of Quest Panel Debug Category Data Structures=@n1) -> (Array Element:Quest Panel Debug Category Data Structure)
   > n4 = Create Entry!(execute, self:Dynamic Entry Box Object Reference=Get EntryBox_Products()) -> (then, ReturnValue:User Widget Object Reference)
   > n5 = Cast To QuestDebugToolCategoryGroup!(execute, Object=@n4) -> (then, CastFailed, AsQuest Debug Tool Category Group:Quest Debug Tool Category Group Object Reference) {is=title K2Node_DynamicCast}
   > SetData!(execute, self:Quest Debug Tool Category Group Object Reference=@n5, Data:"Quest Panel Debug Category Data Structure (by ref)"=@n3) -> (then) {is=title K2Node_CallFunction}
   SetActiveWidget(execute=@n2.else, self=Get Switcher_Content(), Widget=Get Text_NoProducts() -> (Text_NoProducts:Common Text Object Reference))
  ]}
]
variables [Header {type=string, default=Header}]
widgets {
 info {inheritedWidgetCount=0, parentClass=&UserWidget, rootWidgetClass=&Overlay, widgetCount=8}
 widgets [
  Overlay_0 {widgetClassPath=&Overlay, widget=Overlay_0}
  BG {
   widgetClassPath=&Image
   parent=Overlay_0
   widget=BG
   properties {colorAndOpacity=<rgba 0 0.3184 0.3646 0.1>}
   slot {class=OverlaySlot, properties {horizontalAlignment=HAlign_Fill, verticalAlignment=VAlign_Fill, padding=<ltrb 3 3 3 3>}}
  }
  VerticalBox_28 {
   widgetClassPath=&VerticalBox
   parent=Overlay_0
   widget=VerticalBox_28
   slot {class=OverlaySlot, properties {horizontalAlignment=HAlign_Fill, verticalAlignment=VAlign_Fill, padding=<ltrb 5 5 5 5>}}
  }
  ScaleBox_0 {widgetClassPath=&ScaleBox, parent=VerticalBox_28, widget=ScaleBox_0, slot {class=VerticalBoxSlot}}
  Text_Header {
   widgetClassPath=&CommonTextBlock
   bIsVariable=true
   parent=ScaleBox_0
   widget=Text_Header
   properties {
    style=/QuestUI/DebugTool/TextStyles/TextStyle_DebugTool_Header.TextStyle_DebugTool_Header_C
    text=Header
    font {
     bForceMonospaced=false
     bMaterialIsStencil=false
     fontMaterial=null
     fontObject=&NotoSans
     letterSpacing=0
     monospacedWidth=1
     size=9
     skewAmount=0
     typefaceFontName=NotoSans-Bold
     outlineSettings {
      bApplyOutlineToDropShadows=false
      bMiteredCorners=false
      bSeparateFillAlpha=false
      outlineMaterial=null
      outlineSize=0
      outlineColor=<rgba 0 0 0 1>
     }
    }
    shadowOffset=<0 0>
   }
   slot {class=ScaleBoxSlot, properties {horizontalAlignment=HAlign_Fill, verticalAlignment=VAlign_Fill}}
  }
  Switcher_Content {
   widgetClassPath=&CommonVisibilitySwitcher
   bIsVariable=true
   parent=VerticalBox_28
   widget=Switcher_Content
   slot {class=VerticalBoxSlot, properties {size {sizeRule=Fill, value=1}}}
  }
  EntryBox_Products {
   widgetClassPath=&DynamicEntryBox
   bIsVariable=true
   parent=Switcher_Content
   widget=EntryBox_Products
   properties {
    entryBoxType=Vertical
    entryWidgetClass=&/QuestDebugToolCategoryGroup.QuestDebugToolCategoryGroup_C
    numDesignerPreviewEntries=2
    entrySpacing=<0 10>
   }
   slot {class=CommonVisibilitySwitcherSlot}
  }
  Text_NoProducts {
   widgetClassPath=&CommonTextBlock
   bIsVariable=true
   parent=Switcher_Content
   widget=Text_NoProducts
   properties {
    autoWrapText=true
    style=/QuestUI/DebugTool/TextStyles/TextStyle_DebugTool_Body.TextStyle_DebugTool_Body_C
    text=The tags for this quest don't match any existing categories.
    colorAndOpacity {colorUseRule=UseColor_Specified, specifiedColor=<rgba 1 0.9301 0 1>}
    font {
     bForceMonospaced=false
     bMaterialIsStencil=false
     fontMaterial=null
     fontObject=&NotoSans
     letterSpacing=0
     monospacedWidth=1
     size=12.75
     skewAmount=0
     typefaceFontName=NotoSans-Bold
     outlineSettings {
      bApplyOutlineToDropShadows=false
      bMiteredCorners=false
      bSeparateFillAlpha=false
      outlineMaterial=null
      outlineSize=0
      outlineColor=<rgba 0 0 0 1>
     }
    }
    shadowOffset=<0 0>
   }
   slot {class=CommonVisibilitySwitcherSlot}
  }
 ]
}
