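# Quest debug tool widget that shows a quest's Organization section: quest tags, bundle tags and the categories it is available in.
# A visibility switcher holds the category panel and a fallback field, with a helper picking the fallback when organization evaluation was skipped.
# Other helpers turn tag containers into text or "(None Found)" and collapse the playlist product-group alert when the category is available.
# The event graph is empty.
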WidgetBlueprint /QuestUI/DebugTool/QuestWidgets/QuestDebugToolOrganizationInfo
graphs [
 EventGraph {links [], nodes []}
 GetActiveProductWidget {nodes [
   n1 = GetActiveProductWidget!() -> (then, SkippedOrganization:bool=~) {is=title K2Node_FunctionEntry}
   > Return Node!(
    execute
    Widget:Widget Object Reference=Select(
     Option 0:Widget Object Reference=Get Panel_Categories() -> (Panel_Categories:Vertical Box Object Reference)
     Option 1:Widget Object Reference=Get CategoriesNotAvailable() -> (CategoriesNotAvailable:Quest Debug Tool Info Field Object Reference)
     Index=@n1
    ) -> (ReturnValue:Widget Object Reference)
   )
  ]}
 GetWarningVisibility {nodes [
   n1 = GetWarningVisibility!() -> (then, AvailableInActiveProduct:bool=~) {is=title K2Node_FunctionEntry}
   > Return Node!(
    execute
    Visibility:ESlateVisibility Enum=Select(Option 0:ESlateVisibility Enum=HitTestInvisible, Option 1:ESlateVisibility Enum=Collapsed, Index=@n1) -> (ReturnValue:ESlateVisibility Enum)
   )
  ]}
 TagsToText {nodes [
   n1 = TagsToText!() -> (then, TagContainer:Gameplay Tag Container Structure) {is=title K2Node_FunctionEntry}
   n2 = GetDebugStringFromGameplayTagContainer!(TagContainer:"Gameplay Tag Container Structure (by ref)"=@n1) -> (ReturnValue:String)
   n3 = GetNumGameplayTagsInContainer!(TagContainer:"Gameplay Tag Container Structure (by ref)"=@n1) -> (ReturnValue:int=0)
   > Return Node!(
    execute
    Output:Text=Select(Option 0:Text="LOCTEXT(\"(None Found)\")", Option 1:Text="To Text (String)"(InString=@n2), Index="integer > integer"(A=@n3)) -> (ReturnValue:Text)
   )
  ]}
]
variables []
widgets {
 info {inheritedWidgetCount=0, parentClass=&UserWidget, rootWidgetClass=&VerticalBox, widgetCount=9}
 widgets [
  VerticalBox_79 {widgetClassPath=&VerticalBox, widget=VerticalBox_79}
  SectionHeader {
   widgetClassPath=/QuestUI/DebugTool/CommonWidgets/QuestDebugToolSectionHeader
   bIsVariable=true
   parent=VerticalBox_79
   widget=SectionHeader
   properties {header=Organization}
   slot {class=VerticalBoxSlot}
  }
  QuestTags {
   widgetClassPath=/QuestUI/DebugTool/CommonWidgets/QuestDebugToolInfoField
   bIsVariable=true
   parent=VerticalBox_79
   widget=QuestTags
   properties {header=Quest Tags}
   slot {class=VerticalBoxSlot}
  }
  BundleTags {
   widgetClassPath=&QuestDebugToolInfoField
   bIsVariable=true
   parent=VerticalBox_79
   widget=BundleTags
   properties {header=Bundle Tags}
   slot {class=VerticalBoxSlot}
  }
  Switcher_Categories {
   widgetClassPath=&CommonVisibilitySwitcher
   bIsVariable=true
   parent=VerticalBox_79
   widget=Switcher_Categories
   slot {class=VerticalBoxSlot}
  }
  CategoriesNotAvailable {
   widgetClassPath=&QuestDebugToolInfoField
   bIsVariable=true
   parent=Switcher_Categories
   widget=CategoriesNotAvailable
   properties {data=Organizational information was not evaluated due to the quest not being visible., header=Available In}
   slot {class=CommonVisibilitySwitcherSlot}
  }
  Panel_Categories {
   widgetClassPath=&VerticalBox
   bIsVariable=true
   parent=Switcher_Categories
   widget=Panel_Categories
   slot {class=CommonVisibilitySwitcherSlot}
  }
  Widget_Categories {
   widgetClassPath=/QuestUI/DebugTool/CategoryWidgets/QuestDebugToolCategoryField
   bIsVariable=true
   parent=Panel_Categories
   widget=Widget_Categories
   properties {header=Available In}
   slot {class=VerticalBoxSlot}
  }
  Widget_VisibilityAlert {
   widgetClassPath=&QuestDebugToolInfoField
   bIsVariable=true
   parent=Panel_Categories
   widget=Widget_VisibilityAlert
   properties {
    data="Category not viewable within your active playlist's quest product group!"
    displayAsAlert=true
    header=Visibility Alert
    multiline=true
   }
   slot {class=VerticalBoxSlot, properties {size {sizeRule=Fill, value=1}}}
  }
 ]
}
