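# Quest debug tool widget with buttons to grant, progress, complete or track a quest through console commands.
# Execute Command fills text placeholders with quest name and template ID from QuestPanelDebugToolViewModel and quantity from the input field.
# Debug Break buttons briefly set a debug template around a quest rebuild to break in the component or view model.
# Get Widget Switcher Active Widget selects the frontend notice, completed/claimed notice or actions panel.
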WidgetBlueprint /QuestUI/DebugTool/QuestWidgets/QuestDebugToolQuestActions
graphs [
 EventGraph {nodes [
   "On Clicked (Button_GrantQuest)"!() -> (OutputDelegate, then, Button:Common Button Base Object Reference) {is=title K2Node_ComponentBoundEvent}
   > Execute Command!(execute, self, Command:Text="INVTEXT(\"cheat GrantQuest {QuestName}\")") -> (then)
   "On Clicked (Button_ProgressQuest)"!() -> (OutputDelegate, then, Button:Common Button Base Object Reference) {is=title K2Node_ComponentBoundEvent}
   > Execute Command(Command="INVTEXT(\"cheat ProgressAllQuestsByName {QuestName} {Quantity}\")")
   "On Clicked (Button_CompleteQuest)"!() -> (OutputDelegate, then, Button:Common Button Base Object Reference) {is=title K2Node_ComponentBoundEvent}
   > Execute Command(Command="INVTEXT(\"cheat CompleteAllQuestsByName {QuestName}\")")
   "On Clicked (Button_TrackQuest)"!() -> (OutputDelegate, then, Button:Common Button Base Object Reference) {is=title K2Node_ComponentBoundEvent}
   > Execute Command(Command="LOCTEXT(\"TrackQuest {QuestName}\")")
   "On Clicked (Button_DebugComponent)"!() -> (OutputDelegate, then, Button:Common Button Base Object Reference) {is=title K2Node_ComponentBoundEvent}
   > Execute Command(Command="LOCTEXT(\"QuestOrganization.DebugQuestTemplate {TemplateId}\")")
   > Execute Command(Command="LOCTEXT(\"QuestOrganization.Rebuild\")")
   > Execute Command(Command="LOCTEXT(\"QuestOrganization.DebugQuestTemplate none\")")
   "On Clicked (Button_DebugVM)"!() -> (OutputDelegate, then, Button:Common Button Base Object Reference) {is=title K2Node_ComponentBoundEvent}
   > Execute Command(Command="LOCTEXT(\"QuestUI.DebugQuestTemplate {TemplateId}\")")
   > Execute Command(Command="LOCTEXT(\"QuestOrganization.Rebuild\")")
   > Delay(Duration=0.2)
   > Execute Command(Command="LOCTEXT(\"QuestUI.DebugQuestTemplate none\")")
  ]}
 Execute Command {nodes [
   n1 = Execute Command!() -> (then, Command:Text) {is=title K2Node_FunctionEntry}
   n2 = GetQuestName!() -> (QuestName:String)
   n3 = GetQuantity!() -> (Quantity:Text)
   n4 = GetTemplateId!() -> (QuestName:String)
   > ExecuteConsoleCommand!(
    execute
    Command="To String (Text)"(InText=Format Text!(Format=@n1, QuestName:String=@n2, Quantity:Text=@n3, TemplateId:String=@n4) -> (Result))
    SpecificPlayer=GetOwningPlayer()
   ) -> (then)
  ]}
 Get Widget Switcher Active Widget {nodes [
   n1 = Get Widget Switcher Active Widget!() -> (then, QuestState:EFortQuestState Enum=Inactive) {is=title K2Node_FunctionEntry}
   > n2 = "IsA ( soft )"!(execute, Object:Object Reference=GetOwningPlayer(), SoftClass:Object Soft Class Reference=/Script/FortniteGame.FortPlayerControllerFrontEnd) -> (True:Exec, False:Exec)
   n3 = Branch(
    execute=@n2.False
    Condition=OR Boolean(A="Equal (Enum)"(A:EFortQuestState Enum=@n1, B:EFortQuestState Enum=Completed), B="Equal (Enum)"(A=@n1, B=Claimed))
   )
   > Return Node!(execute, Widget:Widget Object Reference=Get Panel_Completed() -> (Panel_Completed:Overlay Object Reference))
   Return Node(execute=@n3.else, Widget=Get Panel_Actions() -> (Panel_Actions:Horizontal Box Object Reference))
   Return Node(execute=@n2.True, Widget=Get Panel_Frontend() -> (Panel_Frontend:Overlay Object Reference))
  ]}
 GetQuantity {nodes [
   GetQuantity() {is=title K2Node_FunctionEntry}
   n1 = Get InputField_ProgressQuantity() -> (InputField_ProgressQuantity:WBP UIKit Input Field Object Reference)
   > Return Node!(
    execute
    Quantity:Text=Get Text(self:Editable Text Object Reference=GetEditableText!(self:WBP UIKit Input Field Object Reference=@n1) -> (NewParam:UIKit Editable Text Object Reference))
   )
  ]}
 GetQuestName {nodes [
   GetQuestName() {is=title K2Node_FunctionEntry}
   n1 = Get QuestDefinition!(self:Quest Panel Debug Tool View Model Object Reference=Get QuestPanelDebugToolViewModel() -> (QuestPanelDebugToolViewModel:Quest Panel Debug Tool View Model Object Reference)) -> (QuestDefinition:Fort Quest Item Definition Object Reference)
   > n2 = Is Valid(InputObject=@n1)
   Return Node!(execute=@n2.Is Not Valid, QuestName:String)
   > Return Node(QuestName=GetObjectName(Object=@n1))
  ]}
 GetTemplateId {nodes [
   GetTemplateId() {is=title K2Node_FunctionEntry}
   > Return Node(QuestName=Get QuestListItemId!(self:Quest Panel Debug Tool View Model Object Reference=Get QuestPanelDebugToolViewModel()) -> (QuestListItemId:String))
  ]}
]
variables []
widgets {
 info {inheritedWidgetCount=0, parentClass=&UserWidget, rootWidgetClass=&Overlay, widgetCount=19}
 widgets [
  Overlay_0 {widgetClassPath=&Overlay, widget=Overlay_0}
  BG {
   widgetClassPath=&Image
   parent=Overlay_0
   widget=BG
   properties {colorAndOpacity=<rgba 0 0.0546 0.0625 0.5>}
   slot {class=OverlaySlot, properties {horizontalAlignment=HAlign_Fill, verticalAlignment=VAlign_Fill, padding=<ltrb 5 5 5 5>}}
  }
  HorizontalBox_26 {
   widgetClassPath=&HorizontalBox
   parent=Overlay_0
   widget=HorizontalBox_26
   slot {class=OverlaySlot, properties {horizontalAlignment=HAlign_Fill, verticalAlignment=VAlign_Center, padding=<ltrb 10 10 10 10>}}
  }
  Text_Header {
   widgetClassPath=&CommonTextBlock
   bIsVariable=true
   parent=HorizontalBox_26
   widget=Text_Header
   properties {
    bIsScrollingEnabled=false
    style=/FortUILibrary/Styles/Typography/Foundation/TextStyle-Base-S-B.TextStyle-Base-S-B_C
    text="Debug\r\nActions"
    font {
     bForceMonospaced=false
     bMaterialIsStencil=false
     fontMaterial=null
     fontObject=&NotoSans
     letterSpacing=0
     monospacedWidth=1
     size=18
     skewAmount=0
     typefaceFontName=NotoSans-Bold
     outlineSettings {
      bApplyOutlineToDropShadows=false
      bMiteredCorners=false
      bSeparateFillAlpha=false
      outlineMaterial=null
      outlineSize=0
      outlineColor=<rgba 0 0 0 1>
     }
    }
    shadowOffset=<0 0>
   }
   slot {class=HorizontalBoxSlot, properties {horizontalAlignment=HAlign_Left, verticalAlignment=VAlign_Top, padding=<ltrb 0 0 20 0>}}
  }
  Button_DebugComponent {
   widgetClassPath=/FortUILibrary/UIKit/Buttons/WBP_UIKit_Button_Quiet
   bIsVariable=true
   parent=HorizontalBox_26
   widget=Button_DebugComponent
   properties {
    minWidth=32
    secondaryText=Component
    text=Debug Break
    toolTipText=Must be attached with a debugger. Triggers a rebuild and breakpoint within the ClientQuestOrganization component for this quest.
   }
   slot {class=HorizontalBoxSlot, properties {horizontalAlignment=HAlign_Left, verticalAlignment=VAlign_Center, padding=<ltrb 0 0 20 0>}}
  }
  Button_DebugVM {
   widgetClassPath=&WBP_UIKit_Button_Quiet
   bIsVariable=true
   parent=HorizontalBox_26
   widget=Button_DebugVM
   properties {
    minWidth=32
    secondaryText=View Model
    text=Debug Break
    toolTipText=Must be attached with a debugger. Triggers a rebuild and breakpoint within the QuestVM for this quest.
   }
   slot {class=HorizontalBoxSlot, properties {horizontalAlignment=HAlign_Left, verticalAlignment=VAlign_Center, padding=<ltrb 0 0 20 0>}}
  }
  WidgetSwitcher_Content {
   widgetClassPath=&CommonActivatableWidgetSwitcher
   bIsVariable=true
   parent=HorizontalBox_26
   widget=WidgetSwitcher_Content
   slot {class=HorizontalBoxSlot, properties {horizontalAlignment=HAlign_Right, size {sizeRule=Fill, value=1}}}
  }
  Panel_Actions {
   widgetClassPath=&HorizontalBox
   bIsVariable=true
   parent=WidgetSwitcher_Content
   widget=Panel_Actions
   slot {class=WidgetSwitcherSlot}
  }
  HB_WithoutView {
   widgetClassPath=&HorizontalBox
   parent=Panel_Actions
   widget=HB_WithoutView
   slot {class=HorizontalBoxSlot, properties {padding=<ltrb 0 0 20 0>}}
  }
  Button_GrantQuest {
   widgetClassPath=&WBP_UIKit_Button_Quiet
   bIsVariable=true
   parent=HB_WithoutView
   widget=Button_GrantQuest
   properties {minWidth=32, text=Grant Quest, toolTipText=Debug grant an instance of this quest}
   slot {class=HorizontalBoxSlot, properties {verticalAlignment=VAlign_Center}}
  }
  HB_WithView {
   widgetClassPath=&HorizontalBox
   parent=Panel_Actions
   widget=HB_WithView
   slot {class=HorizontalBoxSlot, properties {verticalAlignment=VAlign_Center, padding=<ltrb 0 0 10 0>}}
  }
  InputField_ProgressQuantity {
   widgetClassPath=/FortUILibrary/UIKit/InputField/WBP_UIKit_InputField
   bIsVariable=true
   parent=HB_WithView
   widget=InputField_ProgressQuantity
   properties {hint Text="", icon=null, maxTextCharacters=3, text=1, textJustification=Center, widthOverride {enable=true, value=90}}
   slot {class=HorizontalBoxSlot, properties {verticalAlignment=VAlign_Center, padding=<ltrb 0 0 5 0>}}
  }
  Button_ProgressQuest {
   widgetClassPath=&WBP_UIKit_Button_Quiet
   bIsVariable=true
   parent=HB_WithView
   widget=Button_ProgressQuest
   properties {minWidth=32, text=Progress Quest, toolTipText=Debug progress the quest by the quantity given}
   slot {class=HorizontalBoxSlot, properties {verticalAlignment=VAlign_Center, padding=<ltrb 0 0 20 0>}}
  }
  Button_CompleteQuest {
   widgetClassPath=&WBP_UIKit_Button_Quiet
   bIsVariable=true
   parent=HB_WithView
   widget=Button_CompleteQuest
   properties {minWidth=32, text=Complete Quest, toolTipText=Debug complete the quest}
   slot {class=HorizontalBoxSlot, properties {verticalAlignment=VAlign_Center, padding=<ltrb 0 0 20 0>}}
  }
  Button_TrackQuest {
   widgetClassPath=&WBP_UIKit_Button_Quiet
   bIsVariable=true
   parent=HB_WithView
   widget=Button_TrackQuest
   properties {minWidth=32, text=Track Quest, toolTipText=Track/Untrack this quest}
   slot {class=HorizontalBoxSlot, properties {verticalAlignment=VAlign_Center}}
  }
  Panel_Frontend {widgetClassPath=&Overlay, bIsVariable=true, parent=WidgetSwitcher_Content, widget=Panel_Frontend, slot {class=WidgetSwitcherSlot}}
  Text_Frontend {
   widgetClassPath=&CommonTextBlock
   bIsVariable=true
   parent=Panel_Frontend
   widget=Text_Frontend
   properties {
    style=/FortUILibrary/Styles/Typography/Foundation/TextStyle-Base-S-B-40pc.TextStyle-Base-S-B-40pc_C
    text=Quest debug progression is not available in the frontend
    colorAndOpacity {colorUseRule=UseColor_Specified, specifiedColor=<rgba 1 1 1 0.4>}
    font {
     bForceMonospaced=false
     bMaterialIsStencil=false
     fontMaterial=null
     fontObject=&NotoSans
     letterSpacing=0
     monospacedWidth=1
     size=18
     skewAmount=0
     typefaceFontName=NotoSans-Bold
     outlineSettings {
      bApplyOutlineToDropShadows=false
      bMiteredCorners=false
      bSeparateFillAlpha=false
      outlineMaterial=null
      outlineSize=0
      outlineColor=<rgba 0 0 0 1>
     }
    }
    shadowOffset=<0 0>
   }
   slot {class=OverlaySlot, properties {horizontalAlignment=HAlign_Right, verticalAlignment=VAlign_Center, padding=<ltrb 20 20 10 20>}}
  }
  Panel_Completed {
   widgetClassPath=&Overlay
   bIsVariable=true
   parent=WidgetSwitcher_Content
   widget=Panel_Completed
   slot {class=WidgetSwitcherSlot}
  }
  Text_Completed {
   widgetClassPath=&CommonTextBlock
   bIsVariable=true
   parent=Panel_Completed
   widget=Text_Completed
   properties {
    style=&TextStyle-Base-S-B-40pc_C
    text=This quest is completed and no debug actions are available
    colorAndOpacity {colorUseRule=UseColor_Specified, specifiedColor=<rgba 1 1 1 0.4>}
    font {
     bForceMonospaced=false
     bMaterialIsStencil=false
     fontMaterial=null
     fontObject=&NotoSans
     letterSpacing=0
     monospacedWidth=1
     size=18
     skewAmount=0
     typefaceFontName=NotoSans-Bold
     outlineSettings {
      bApplyOutlineToDropShadows=false
      bMiteredCorners=false
      bSeparateFillAlpha=false
      outlineMaterial=null
      outlineSize=0
      outlineColor=<rgba 0 0 0 1>
     }
    }
    shadowOffset=<0 0>
   }
   slot {class=OverlaySlot, properties {horizontalAlignment=HAlign_Right, verticalAlignment=VAlign_Center, padding=<ltrb 20 20 10 20>}}
  }
 ]
}
