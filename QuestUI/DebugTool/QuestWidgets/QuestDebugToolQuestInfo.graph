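# Quest debug tool widget that lays out a "Quest Info" section with a definition field and Display Name, Progress and State fields.
# It composes reusable debug tool header and field widgets, with an empty event graph and no variables.
# Helper functions format progress as "Achieved / Required" or N/A and convert quest state to text.
# Two more return visibility for query tags and compatibility alerts, though no matching widgets appear in this tree.
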WidgetBlueprint /QuestUI/DebugTool/QuestWidgets/QuestDebugToolQuestInfo
graphs [
 EventGraph {links [], nodes []}
 GetAdditionalQueryTagVisibility {nodes [
   n1 = GetAdditionalQueryTagVisibility!() -> (then, AdditionalQueryTags:Text) {is=title K2Node_FunctionEntry}
   > Return Node!(execute, Visibility:ESlateVisibility Enum=Select(Option 0=HitTestInvisible, Option 1=Collapsed, Index=TextIsEmpty(InText=@n1)))
  ]}
 GetQueryAlertVisibility {nodes [
   n1 = GetQueryAlertVisibility!() -> (then, IsCompatible:bool=~) {is=title K2Node_FunctionEntry}
   > Return Node(Visibility=Select(Option 0=HitTestInvisible, Option 1=Collapsed, Index=@n1))
  ]}
 GetQuestProgress {nodes [
   n1 = GetQuestProgress!() -> (then, Achieved:int, Required:int) {is=title K2Node_FunctionEntry}
   > n2 = Branch(Condition="integer > integer"(A=@n1.Required))
   > Return Node!(execute, Name:Text=Format Text!(Format="LOCTEXT(\"{0} / {1}\")", 0:int=@n1, 1:int=@n1.Required) -> (Result))
   Return Node(execute=@n2.else, Name="LOCTEXT(\"N/A\")")
  ]}
 GetQuestState {nodes [
   n1 = GetQuestState!() -> (then, QuestState:EFortQuestState Enum=Inactive) {is=title K2Node_FunctionEntry}
   > Return Node!(execute, State:Text="To Text (String)"(InString=Enum to String!(Enumerator:EFortQuestState Enum=@n1) -> (ReturnValue:String)))
  ]}
]
variables []
widgets {
 info {inheritedWidgetCount=0, parentClass=&UserWidget, rootWidgetClass=&VerticalBox, widgetCount=8}
 widgets [
  VerticalBox_1 {widgetClassPath=&VerticalBox, widget=VerticalBox_1}
  SectionHeader {
   widgetClassPath=/QuestUI/DebugTool/CommonWidgets/QuestDebugToolSectionHeader
   bIsVariable=true
   parent=VerticalBox_1
   widget=SectionHeader
   properties {header=Quest Info}
   slot {class=VerticalBoxSlot}
  }
  VerticalBox_0 {widgetClassPath=&VerticalBox, parent=VerticalBox_1, widget=VerticalBox_0, slot {class=VerticalBoxSlot}}
  Widget_QuestDefinition {
   widgetClassPath=/QuestUI/DebugTool/CommonWidgets/QuestDebugToolDefinitionField
   bIsVariable=true
   parent=VerticalBox_0
   widget=Widget_QuestDefinition
   properties {assetEntryType=QuestAsset, definitionHeader=Quest Definition, displayDataHeader=Quest Display Data, viewEntryType=QuestView}
   slot {class=VerticalBoxSlot}
  }
  HorizontalBox_1 {widgetClassPath=&HorizontalBox, parent=VerticalBox_0, widget=HorizontalBox_1, slot {class=VerticalBoxSlot}}
  Widget_DisplayName {
   widgetClassPath=/QuestUI/DebugTool/CommonWidgets/QuestDebugToolInfoField
   bIsVariable=true
   parent=HorizontalBox_1
   widget=Widget_DisplayName
   properties {header=Display Name}
   slot {class=HorizontalBoxSlot, properties {size {sizeRule=Fill, value=1}}}
  }
  Widget_Progress {
   widgetClassPath=&QuestDebugToolInfoField
   bIsVariable=true
   parent=HorizontalBox_1
   widget=Widget_Progress
   properties {header=Progress}
   slot {class=HorizontalBoxSlot}
  }
  Widget_State {
   widgetClassPath=&QuestDebugToolInfoField
   bIsVariable=true
   parent=HorizontalBox_1
   widget=Widget_State
   properties {header=State}
   slot {class=HorizontalBoxSlot}
  }
 ]
}
