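# Quest debug tool widget with a Visibility section: a header plus Current State and Reason info fields in a vertical box.
# Helper functions turn a content-visible flag into "Visible"/"NOT VISIBLE" text and flag the not-visible state for alert display.
# Another helper collapses on empty reason text, otherwise returning hit-test invisible.
# The event graph is empty, and the file shows no callers or bindings for these helpers.
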WidgetBlueprint /QuestUI/DebugTool/QuestWidgets/QuestDebugToolVisibilityInfo
graphs [
 EventGraph {links [], nodes []}
 GetCurrentStateData {nodes [
   n1 = GetCurrentStateData!() -> (then, IsContentVisible:bool=~) {is=title K2Node_FunctionEntry}
   > Return Node!(execute, CurrentState:Text=SelectText(A="LOCTEXT(\"Visible\")", B="LOCTEXT(\"NOT VISIBLE\")", bPickA=@n1))
  ]}
 GetCurrentStateDisplayAsAlert {nodes [
   n1 = GetCurrentStateDisplayAsAlert!() -> (then, IsContentVisible:bool=~) {is=title K2Node_FunctionEntry}
   > Return Node!(execute, DisplayAsAlert:bool=NOT Boolean(A=@n1))
  ]}
 GetReasonVisibility {nodes [
   n1 = GetReasonVisibility!() -> (then, ReasonText:Text) {is=title K2Node_FunctionEntry}
   > Return Node!(execute, Visibility:ESlateVisibility Enum=Select(Option 0=HitTestInvisible, Option 1=Collapsed, Index=TextIsEmpty(InText=@n1)))
  ]}
]
variables []
widgets {
 info {inheritedWidgetCount=0, parentClass=&UserWidget, rootWidgetClass=&VerticalBox, widgetCount=4}
 widgets [
  VerticalBox_79 {widgetClassPath=&VerticalBox, widget=VerticalBox_79}
  SectionHeader {
   widgetClassPath=/QuestUI/DebugTool/CommonWidgets/QuestDebugToolSectionHeader
   bIsVariable=true
   parent=VerticalBox_79
   widget=SectionHeader
   properties {header=Visibility}
   slot {class=VerticalBoxSlot}
  }
  CurrentState {
   widgetClassPath=/QuestUI/DebugTool/CommonWidgets/QuestDebugToolInfoField
   bIsVariable=true
   parent=VerticalBox_79
   widget=CurrentState
   properties {header=Current State}
   slot {class=VerticalBoxSlot}
  }
  Reason {
   widgetClassPath=&QuestDebugToolInfoField
   bIsVariable=true
   parent=VerticalBox_79
   widget=Reason
   properties {header=Reason}
   slot {class=VerticalBoxSlot}
  }
 ]
}
